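; Parent Blueprint for adjustable mine rail supports, a Fortnite building piece built from a base mesh, stacked mid sections and an end cap.
; The construction script measures mesh Z bounds, adds at least one instanced mid section and places the end cap at a computed or overridden position.
; When building texture data exists, its tint is applied to the mid section and cap.
; Server death destroys the mid-section component.
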(event EventBeginPlay)

(event Collision|EventActorBeginOverlap
  (params (OtherActor :type Actor%20Object%20Reference)))

(event EventTick
  (params (DeltaSeconds :type Float%20%28single-precision%29)))

(event Custom|DestroyInstances
  (Components|DestroyComponent (Variables|Default|GetMidSection)))

(event EventOnDeathServer
  (params (Damage :type Float%20%28single-precision%29) (DamageTags :type Gameplay%20Tag%20Container%20Structure%20%28by%20ref%29) (Momentum :type Vector :default "0, 0, 0") (HitInfo :type Hit%20Result%20Structure%20%28by%20ref%29) (InstigatedBy :type Controller%20Object%20Reference) (DamageCauser :type Actor%20Object%20Reference) (EffectContext :type Gameplay%20Effect%20Context%20Handle%20Structure))
  (CallFunction|DestroyInstances))

(fn GetMaterialFromRootMesh
  (Rendering|Material|SetMaterial (Variables|Default|GetEndCap) 0 (Material|GetMID (Variables|Debug|GetStaticMeshComponent))))

(fn GetMeshBounds
  (Variables|Default|SetBaseBounds (.z (Components|StaticMesh|GetLocalBounds (Variables|Debug|GetStaticMeshComponent))))
  (Variables|Default|SetEndCapBounds (.z (Components|StaticMesh|GetLocalBounds (Variables|Default|GetEndCap))))
  (Variables|Default|SetMidSectionBounds (.z (Components|StaticMesh|GetLocalBounds (Variables|Default|GetMidSection)))))

(fn GetTintColorFromTextureData
  (bind _outtexturedataarray (Building|GetBuildingTextureData))
  (Variables|Default|SetTintColor (Variables|/Script/FortniteGame.BuildingTextureData|GetTintColor (Utilities|Array|Get(acopy) _outtexturedataarray))))

(fn SetEndCapPosition
  (if (Variables|Default|GetOverrideEndCapPosition)
    (Transformation|SetRelativeLocation (Variables|Default|GetEndCap) (Variables|Default|GetEndCapPosition) false false)
    (else
      (bind _output_get (Variables|Default|SetEndCapPosition (Math|Vector|MakeVector 0.0 0.0 (+ (* (Variables|Default|GetMidSectionBounds) (Variables|Default|GetNumberOfSections)) (Variables|Default|GetEndCapBounds)))))
      (Transformation|SetRelativeLocation (Variables|Default|GetEndCap) _output_get false false))))

(fn SetNumberofSections
  (if (< (Variables|Default|GetNumberOfSections) 1)
    (bind _output_get (Variables|Default|SetNumberOfSections 1))
    (else
      _))
  (for _index (range _output_get)
    (Components|InstancedStaticMesh|AddInstance (Variables|Default|GetMidSection) (Math|Transform|MakeTransform (Math|Vector|MakeVector 0.0 0.0 (+ (Variables|Default|GetBaseBounds) (* _index (Variables|Default|GetMidSectionBounds))))) false)))

(fn SetTintColorOnMeshParts
  (for _array_element (Utilities|Array|MakeArray (Variables|Default|GetMidSection) (Variables|Default|GetEndCap))
    (Default|SetColoronSecondaryMesh (Math|Conversions|ToLinearColor(Color) (Variables|Default|GetTintColor)) _array_element)))

(fn ConstructionScript
  (CallFunction|ValidateMeshComponents)
  (CallFunction|GetMeshBounds)
  (CallFunction|SetNumberofSections)
  (CallFunction|SetEndCapPosition)
  (CallFunction|ValidateBuildingTextureData)
  (if (Variables|Default|GetHasTextureData)
    (CallFunction|GetTintColorfromTextureData)
    (CallFunction|SetTintColoronMeshParts)))

(fn ValidateBuildingTextureData
  (bind _outtexturedataarray (Building|GetBuildingTextureData))
  (Variables|Default|SetHasTextureData (> (Utilities|Array|Length _outtexturedataarray) 0)))

(fn ValidateMeshComponents
  (for _array_element (Utilities|Array|MakeArray (Variables|Debug|GetStaticMeshComponent) (Variables|Default|GetEndCap) (Variables|Default|GetMidSection))
    (Utilities|IsValid _array_element)))
