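# Parent material for the grindable industrial mine rail, enabled for spline meshes and Nanite, with texture samples and procedural emissive light strips.
# Local-position masks shape strips and spline-tiled bulbs, alongside panning noise and fresnel, distance, top-view and cap masks.
# A sphere mask around the collection's PlayerPosition, driven by custom primitive data for sprint, jump-off and idle tail, supports grind highlights.
# Emissive color is divided by eye adaptation.
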Material /GrindRail/Industrial_MineRailSupports_A/Materials/M_Industrial_MineRail_Parent
graph {
 properties {
  bUsedWithNanite=true
  bUsedWithSplineMeshes=true
  previewMesh=/GrindRail/Industrial_MineRailSupports_A/Meshes/SM_Industrial_MineRail_A
  usedShadingModels=Default Lit
 }
 nodes [
  n1 = VectorParameter() {group=Emissive, parameterName=Emissive_Color2, defaultValue=<rgba 0.8 0.056 0.058 1>}
  n2 = VectorParameter() {group=Emissive, parameterName=Emissive_Color1, defaultValue=<rgba 2.5 0.6118 0.075 1>}
  n3 = NamedRerouteUsage()
  n4 = VectorParameter() {group=Emissive, parameterName=Emissive_PlayerGrind_Color2, defaultValue=<rgba 4 0.2 0 1>}
  n5 = VectorParameter() {group=Emissive, parameterName=Emissive_PlayerGrind_Color1, defaultValue=<rgba 50 11.4566 0.5 1>}
  n6 = MaterialFunctionCall(In=NamedRerouteUsage(), Contrast=ScalarParameter() {defaultValue=0.3, group=Emissive, parameterName=Emissive_PlayerGrind_Contrast}) {materialFunction=&CheapContrast}
  n7 = Multiply(
   A=LinearInterpolate(A=LinearInterpolate(A=@n1.RGB, B=@n2.RGB, Alpha=@n3), B=LinearInterpolate(A=@n4.RGB, B=@n5.RGB, Alpha=@n6.Result), Alpha=NamedRerouteUsage()) {constA=1}
   B=NamedRerouteUsage()
  )
  n8 = EyeAdaptation()
  NamedRerouteDeclaration(None=Multiply(A=Divide(A=@n7, B=@n8.EyeAdaptation))) {name=Emissive, nodeColor=<rgba 1 0.1641 0 1>}
  n9 = Divide(
   A=Subtract(A=PixelDepth().R, B=ScalarParameter() {defaultValue=512, group=Emissive, parameterName=M_Distance Min})
   B=ScalarParameter() {defaultValue=768, group=Emissive, parameterName=M_Distance falloff}
  )
  NamedRerouteDeclaration(None=OneMinus(None=Saturate(None=@n9))) {name=M_Distance, nodeColor=<rgba 1 0 0.2812 1>}
  n10 = NamedRerouteUsage()
  n11 = NamedRerouteUsage()
  NamedRerouteDeclaration(None=Multiply(A=Max(A=@n10, B=@n11), B=NamedRerouteUsage())) {name=M_FresnelDist, nodeColor=<rgba 0 0.6016 1 1>}
  n12 = Constant() {r=10}
  n13 = Constant() {r=2}
  n14 = NamedRerouteDeclaration(None=VertexInterpolator(VS=LocalPosition() {localOrigin=InstancePreSkinning}.XYZ).PS) {name=LocalPos_PS, nodeColor=<rgba 0.6172 0 1 1>}
  n15 = Abs(None=@n14)
  NamedRerouteDeclaration(None=SmoothStep(Min=Add(A=@n12, B=@n13) {constB=0}, Max=@n12, Value=ComponentMask(None=@n15) {r=true}) {constMax=10, constMin=13}) {name=M_Inner/OuterStrip, nodeColor=<rgba 1 0 0.9375 1>}
  n16 = WorldPosition()
  n17 = CollectionParameter() {collection=&FortniteMaterialParameters, parameterName=PlayerPosition}
  n18 = ComponentMask(None=@n17) {b=true, g=true, r=true}
  n19 = LinearInterpolate(
   A=LinearInterpolate(
    A=ScalarParameter() {defaultValue=250, group=Emissive, parameterName=RadiusDefaultEffect}
    B=ScalarParameter() {defaultValue=500, group=Emissive, parameterName=RadiusSprintingEffectRadius}
    Alpha=ScalarParameter() {bUseCustomPrimitiveData=true, group=Gameplay, parameterName=sprinting state, primitiveDataIndex=8}
   )
   Alpha=ScalarParameter() {bUseCustomPrimitiveData=true, defaultValue=1, parameterName=IdleTail light switch, primitiveDataIndex=6}
  ) {constB=0}
  n20 = ScalarParameter() {defaultValue=0.3, group=Emissive, parameterName=HardnessRadius}
  n21 = Multiply(A=Power(Base=SphereMask(A=@n16.XYZ, B=@n18, Radius=@n19, Hardness=@n20)), B=Constant() {r=1.02})
  n22 = ScalarParameter() {bUseCustomPrimitiveData=true, group=Gameplay, parameterName=Jumpoff, primitiveDataIndex=7}
  n23 = ScalarParameter() {bUseCustomPrimitiveData=true, group=Gameplay, parameterName=sprinting state, primitiveDataIndex=8}
  NamedRerouteDeclaration(None=Multiply(A=LinearInterpolate(A=@n21, Alpha=@n22) {constB=0}, B=LinearInterpolate(Alpha=@n23) {constA=0.7})) {name=Player SphereMask, nodeColor=<rgba 0.5 0 1 1>}
  n24 = NamedRerouteUsage()
  n25 = ComponentMask(None=Abs(None=Add(A=@n24) {constB=-6.5})) {g=true}
  n26 = ComponentMask(None=Abs(None=@n24)) {r=true}
  NamedRerouteDeclaration(None=Multiply(A=Step(Y=@n25) {constX=6, constY=0.9}, B=Step(Y=@n26) {constX=16, constY=0.9})) {name=M_SafeArea, nodeColor=<rgba 0.9219 1 0 1>}
  n27 = Add(
   A=ComponentMask(None=MaterialFunctionCall() {materialFunction=&CameraDirectionVector}.Result) {b=true}
   B=Constant() {r=0.8}
  )
  n28 = Constant() {r=-3}
  NamedRerouteDeclaration(None=Saturate(None=Multiply(A=@n27, B=@n28))) {name=M_TopView, nodeColor=<rgba 1 0 0.5859 1>}
  n29 = ComponentMask(None=Add(A=NamedRerouteUsage()) {constB=256}) {b=true}
  n30 = Add(A=Multiply(A=Abs(None=@n29)) {constB=-1}, B=ScalarParameter() {defaultValue=230, group=Emissive, parameterName=Emissive_CapMask Size})
  n31 = ScalarParameter() {defaultValue=5, group=Emissive, parameterName=Emissive_CapMask Smothness}
  NamedRerouteDeclaration(None=Saturate(None=Divide(A=@n30, B=@n31))) {name=M_Caps, nodeColor=<rgba 0 0.6484 1 1>}
  n32 = Time()
  n33 = Constant() {r=-5}
  n34 = MaterialFunctionCall(Color=Divide(A=NamedRerouteUsage()) {constB=512}) {materialFunction=&SplitComponents}
  n35 = AppendVector(A=@n34.G, B=@n34.G)
  n36 = ScalarParameter() {defaultValue=0.1, group=Emissive, parameterName=Emissive_Anim_1_Tile_X}
  n37 = Multiply(A=ScalarParameter() {defaultValue=2, group=Emissive, parameterName=Emissive_Anim_1_Tile_Y}, B=NamedRerouteUsage())
  n38 = Add(A=AppendVector(A=Constant(), B=Divide(A=@n32, B=@n33)), B=Multiply(A=@n35, B=AppendVector(A=@n36, B=@n37)) {constB=0.1})
  n39 = AppendVector(
   A=ScalarParameter() {defaultValue=0.05, group=Emissive, parameterName=Emissive_Anim_2_Tile_X}
   B=Multiply(A=ScalarParameter() {defaultValue=3, group=Emissive, parameterName=Emissive_Anim_2_Tile_Y}, B=NamedRerouteUsage())
  )
  n40 = Constant()
  n41 = Divide(A=Time(), B=ScalarParameter() {defaultValue=2, group=Emissive, parameterName=Emissive_Anim_1_speed})
  n42 = Frac(None=Add(A=Multiply(A=@n35, B=@n39) {constB=0.5}, B=AppendVector(A=@n40, B=@n41)))
  NamedRerouteDeclaration(None=Max(
    A=TextureSample(UVs=@n38) {samplerType=SAMPLERTYPE_Grayscale, texture=/Game/Global/Textures/Default/Noise/LowResBlurredNoiseNormalized}.R
    B=TextureSample(UVs=@n42) {samplerType=SAMPLERTYPE_Grayscale, texture=&LowResBlurredNoiseNormalized}.R
   )) {name=Animation, nodeColor=<rgba 1 0.375 0 1>}
  n43 = ScalarParameter() {defaultValue=-3.2, group=Emissive, parameterName=Emissive_Offset Z_Outer Strip}
  n44 = ScalarParameter() {defaultValue=-4, group=Emissive, parameterName=Emissive_Offset Z_Inner Strip}
  n45 = NamedRerouteUsage()
  n46 = MaterialFunctionCall(
   R=ScalarParameter() {group=Emissive, parameterName=Emissive_Offset X}
   G=ShadingPathSwitch(
    Default=LinearInterpolate(A=@n43, B=@n44, Alpha=@n45)
    Mobile=ScalarParameter() {defaultValue=-8.5, group=Emissive, parameterName=Emissive_Offset Z_Mobile}
   )
   B=ScalarParameter() {group=Emissive, parameterName=Emissive_Offset Y}
  ) {materialFunction=&AppendMany}
  n47 = ScalarParameter() {defaultValue=18, group=Emissive, parameterName=Emissive_Size X}
  n48 = LinearInterpolate(
   A=ScalarParameter() {defaultValue=2, group=Emissive, parameterName=Emissive_Size Z_Outer Strip}
   B=ScalarParameter() {defaultValue=2.7, group=Emissive, parameterName=Emissive_Size Z_Iner Strip}
   Alpha=NamedRerouteUsage()
  )
  n49 = ScalarParameter() {defaultValue=0.5, group=Emissive, parameterName=Emissive_Size Y}
  n50 = MaterialFunctionCall!(
   A=Divide(
    A=Add(A=@n14, B=@n46.RGB)
    B=MaterialFunctionCall(R=@n47, G=@n48, B=@n49) {materialFunction=&AppendMany}.RGB
   )
   B
   Bounds=ScalarParameter() {defaultValue=0.5, group=Emissive, parameterName=Bounds}
   Edge Falloff=ScalarParameter() {defaultValue=1.5, group=Emissive, parameterName=Falloff}
  ) {materialFunction=/Engine/Functions/Engine_MaterialFunctions01/Gradient/BoxMask-2D}
  n51 = NamedRerouteUsage()
  NamedRerouteDeclaration(None=Multiply(A=Multiply(A=@n50.Result, B=@n51), B=NamedRerouteUsage())) {name=M_LightStrips, nodeColor=<rgba 1 0 0.5156 1>}
  n52 = NamedRerouteUsage()
  n53 = ScalarParameter() {defaultValue=4, group=Emissive, parameterName=Light Freq}
  n54 = NamedRerouteDeclaration(None=ScalarParameter() {bUseCustomPrimitiveData=true, defaultValue=1, parameterName=Tiling_V, primitiveDataIndex=18}) {name=P_SplineLenght, nodeColor=<rgba 0 1 0.9297 1>}
  n55 = Sine(None=Divide(A=ComponentMask(None=@n52) {b=true}, B=Divide(A=@n53, B=@n54)) {constB=20})
  n56 = ScalarParameter() {defaultValue=0.5, group=Emissive, parameterName=Lights gap}
  n57 = NamedRerouteUsage()
  n58 = Multiply(A=Multiply(A=Add(A=@n55, B=@n56), B=Multiply(A=@n57, B=@n57)), B=NamedRerouteUsage())
  n59 = ScalarParameter() {defaultValue=-0.2, group=Emissive, parameterName=Emissive_Lights Contrast}
  NamedRerouteDeclaration(None=LinearInterpolate(
    A=NamedRerouteUsage()
    B=MaterialFunctionCall(In=@n58, Contrast=@n59) {materialFunction=&CheapContrast}.Result
    Alpha=NamedRerouteUsage()
   )) {name=M_light Bulbs, nodeColor=<rgba 0 0.8359 1 1>}
  n60 = CameraVectorWS()
  n61 = VertexNormalWS()
  n62 = Add(
   A=NamedRerouteDeclaration(None=DotProduct(A=@n60, B=@n61)) {name=M_Fresnel, nodeColor=<rgba 0.3828 1 0 1>}
   B=ScalarParameter() {defaultValue=-0.2, group=Emissive, parameterName=M_Fresnel bias}
  )
  n63 = ScalarParameter() {defaultValue=25, group=Emissive, parameterName=M_Fresnel scale}
  NamedRerouteDeclaration(None=Saturate(None=Multiply(A=@n62, B=@n63))) {name=M_Fresnel2, nodeColor=<rgba 0 1 0.6016 1>}
  NamedRerouteUsage()
  TextureSampleParameter2D() {parameterName=Diffuse, texture=/GrindRail/Industrial_MineRailSupports_A/Textures/T_Industrial_MineRail_A_A_D}
  TextureSampleParameter2D() {parameterName=Normals, samplerType=SAMPLERTYPE_Normal, texture=/GrindRail/Industrial_MineRailSupports_A/Textures/T_Industrial_MineRail_A_A_N}
  TextureSampleParameter2D() {
   desc="R = Specular\r\nG = Metal\r\nB = Roughness"
   parameterName=SpecularMasks
   samplerType=SAMPLERTYPE_Masks
   texture=/GrindRail/Industrial_MineRailSupports_A/Textures/T_Industrial_MineRail_A_A_S
  }
 ]
}
